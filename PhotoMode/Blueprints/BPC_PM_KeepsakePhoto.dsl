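; Photo Mode component that marks its owner, a player character or a place, as a keepsake photo subject and reports when it is photographed.
; For pawns, it takes the keepsake name from the equipped character and looks up that character's verb tag in a data table, when cosmetics apply and on server capture.
; The server broadcasts that custom verb; the client shows a HUD message, using the tag suffix if no name is set.
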(event EventBeginPlay
  (bind _asfort_player_pawn (Utilities|Casting|CastToFortPlayerPawn (Components|GetOwner)))
  (Cosmetics|WhenCosmeticsApplied _asfort_player_pawn (delegate Custom|OnCosmeticsApplied)))

(event Custom|OnCosmeticsApplied
  (params (Pawn :type Fort%20Player%20Pawn%20Object%20Reference) (BaseMesh :type Skeletal%20Mesh%20Component%20Object%20Reference) (PartsArray :type Array%20of%20Custom%20Character%20Part%20Object%20References) (ValidSkeletalMeshComponentPartArray :type Array%20of%20Skeletal%20Mesh%20Component%20Object%20References))
  (CallFunction|TranslateCharacterID Pawn))

(fn TakeKeepsakePhotoOnClient
  (params (Scene%20Capture%20Component :type Scene%20Capture%20Component%202D%20Object%20Reference))
  (bind (_lefts _rights) (Utilities|String|Split (Utilities|String|ToString(Name) (GameplayTags|GetTagName (Variables|Required|GetPhotoTakenVerb))) "CustomVerb.PhotoMode.Photo.Taken."))
  (GameplayEventRouter|BroadcastEvent (GameplayEventRouter|GetGlobalEventRouter) (Utilities|Struct|MakeHUDMessageData "TopCenter" 0 0 "None" 0 "None" (Utilities|Text|FormatText "NSLOCTEXT(\"[B47BAB45EEADFAA86021B68F65B6CFB6]\", \"6E1A3C644CF0AE9EDAEA8CBBE0CF8ED3\", \"{KeepsakeName} Keepsake Photo Taken\")" (Utilities|String|SelectText (Utilities|Text|ToText(String) _rights) (Variables|Optional|GetKeepsake%20Name) (Utilities|Text|TextIsEmpty (Variables|Optional|GetKeepsake%20Name)))) 0 0 0.0 false 0 7.0) 0))

(fn TakeKeepsakePhotoOnServer
  (params (PhotoTaker :type Player%20Controller%20Object%20Reference))
  ;[0.2031 0.3993 1 0.5] Server Things
  ;; Check if Character Otherwise its a place
  (bind _asfort_player_pawn (Utilities|Casting|CastToFortPlayerPawn (Components|GetOwner))
    (:then
      (CallFunction|TranslateCharacterID _asfort_player_pawn))
    (:CastFailed
      _))
  (CustomVerb|BroadcastCustomVerbMessage (CustomVerb|MakeCustomVerbMessage PhotoTaker (Variables|Required|GetPhotoTakenVerb))))

(fn TranslateCharacterID
  (params (Player%20Pawn :type Fort%20Player%20Pawn%20Object%20Reference))
  (bind (_returnvalue_bannericonid _returnvalue_bannercolorid _returnvalue_skydivecontrail _returnvalue_glider _returnvalue_pickaxe _returnvalue_character) (Utilities|Struct|BreakFortAthenaLoadout (Cosmetics|GetPawnCosmeticLoadout Player%20Pawn false true)))
  (bind _returnvalue (ItemDefinition|GetDisplayName _returnvalue_character))
  (bind (_returnvalue_bannericonid_2 _returnvalue_bannercolorid_2 _returnvalue_skydivecontrail_2 _returnvalue_glider_2 _returnvalue_pickaxe_2 _returnvalue_character_2) (Utilities|Struct|BreakFortAthenaLoadout (Cosmetics|GetPawnCosmeticLoadout (Variables|GetPlayer%20Pawn) false true)))
  (bind _keepsake_data_table (Variables|Default|GetKeepsake%20Data%20Table))
  (Utilities|IsValid _returnvalue_character
    (:Is%20Valid
      (if (not (Utilities|Text|TextIsEmpty _returnvalue))
        (Variables|Optional|SetKeepsake%20Name _returnvalue)
        (Variables|Required|SetPhotoTakenVerb "(TagName=\"CustomVerb.PhotoMode.Photo.Taken\")")
        (bind _outrownames (DataTable|GetDataTableRowNames _keepsake_data_table))
        (for _array_element _outrownames
          (bind _returnvalue_2 (Utilities|GetDataTableRow _keepsake_data_table _array_element))
          (if (== (Utilities|Struct|BreakPhotoModeKeepsakePhotoStruct _returnvalue_2) _returnvalue_character_2)
            (Variables|Required|SetPhotoTakenVerb (.CustomVerb_5_E82764D94D3019861D0364AB5A7DB727 (Utilities|Struct|BreakPhotoModeKeepsakePhotoStruct _returnvalue_2)))))))))
