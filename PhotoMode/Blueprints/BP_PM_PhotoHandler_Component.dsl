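; Photo mode component that takes pictures on the client and manages stored photo textures and tags.
; Capture events prepare paths and resolution, request a scene capture or screenshot, then a one-shot OnScreenshotReady binding broadcasts the texture and last metadata to UI through GameplayEventRouter.
; Helpers derive FOV by scaling the camera's half-angle tangent for orientation against 16:9, import textures by GUID, and edit tags by re-adding rebuilt metadata to player photo storage.
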(event Custom|TakeCapturePicture_Client
  (params (MetaData :type Photo%20Metadata%20Structure) (SceneCapture :type Scene%20Capture%20Component%202D%20Object%20Reference) (LocalPlayerController :type Player%20Controller%20Object%20Reference))
  (Variables|Default|SetLocalPlayerController LocalPlayerController)
  (bind (_xresolution _yresolution _savepath _folderpath _filepath) (CallFunction|PreparePhotoData MetaData))
  (EventDispatchers|BindEventtoOnScreenshotReady (delegate Custom|OnScreenshotTaken))
  (CameraPhoto|RequestSceneCapture _xresolution _yresolution _folderpath _filepath SceneCapture))

(event Custom|OnScreenshotTaken
  (params (Width :type Integer) (Height :type Integer) (Texture :type Texture%202D%20Object%20Reference))
  ; Screenshot Taken Callback
  (EventDispatchers|UnbindEventfromOnScreenshotReady (delegate Custom|OnScreenshotTaken))
  (GameplayEventRouter|BroadcastEvent (GameplayEventRouter|GetGlobalEventRouter) (Utilities|Struct|MakeSPMPhotoCameraPhotoTakenUIMessage Texture (Variables|Default|GetLastMetaData)) (Variables|Default|GetLocalPlayerController)))

(event Custom|TakeScreenshotPicture_Client
  (params (MetaData :type Photo%20Metadata%20Structure) (LocalPlayerController :type Player%20Controller%20Object%20Reference))
  (Variables|Default|SetLocalPlayerController LocalPlayerController)
  (bind (_xresolution _yresolution _savepath) (CallFunction|PreparePhotoData MetaData))
  (EventDispatchers|BindEventtoOnScreenshotReady (delegate Custom|OnScreenshotTaken))
  (CameraPhoto|RequestScreenshot _xresolution _yresolution _savepath))

(fn CalculateFOV
  (params (PhotoOrientation :type EPhotoOrientation%20Enum :default "Square"))
  (return (* (Math|Trig|Atan(Degrees) (* (Math|Trig|Tan(Degrees) (* 0.5 (Camera|GetFOVAngle (Game|GetPlayerCameraManager 0)))) (/ (Utilities|Select :Index (Variables|GetPhotoOrientation) :Square 1.0 :Landscape (/ 16.0 9.0) :Portrait (/ 9.0 16.0)) (/ 16.0 9.0)))) 2.0)))

(fn GetPictureTexture
  (params (GUID :type Guid%20Structure))
  (bind _returnvalue (Rendering|ImportFileasTexture2D (PhotoMode|GetPhotoPath (CallFunction|GetPhotoStorage (Variables|Default|GetLocalPlayerController)) GUID)))
  (return _returnvalue))

(fn SetPhotoTagEnabled
  (params (GUID :type Guid%20Structure) (Tag :type Gameplay%20Tag%20Structure) (Enabled :type Boolean))
  (bind _localplayercontroller (Variables|Default|GetLocalPlayerController))
  (bind _returnvalue (CallFunction|GetPhotoStorage _localplayercontroller))
  (bind _guid (Variables|GetGUID))
  (bind _photos (Variables|/Script/PhotoModeRuntime.PhotoStoragePlayerStateComponent|GetPhotos _returnvalue))
  (bind (_value _returnvalue_2) (Utilities|Map|Find _photos _guid))
  (bind (_value_guid _value_orientation _value_gameplaytags) (Utilities|Struct|BreakPhotoMetadata _value))
  (bind _newgameplaytags (LocalVariables|GetNewGameplayTags))
  (bind _orientation (LocalVariables|GetOrientation))
  (bind _foundguid (LocalVariables|GetFoundGUID))
  (bind _photometadata (Utilities|Struct|MakePhotoMetadata _foundguid _orientation _newgameplaytags))
  (bind _newgameplaytags_2 (LocalVariables|GetNewGameplayTags))
  (bind _tag (Variables|GetTag))
  (if _returnvalue_2
    (LocalVariables|SetFoundGUID _value_guid)
    (LocalVariables|SetOrientation _value_orientation)
    (LocalVariables|SetNewGameplayTags _value_gameplaytags)
    (if (Variables|GetEnabled)
      (GameplayTags|AddGameplayTag _newgameplaytags_2 _tag :!class /Script/GameplayTags.BlueprintGameplayTagLibrary)
      (else
        (GameplayTags|RemoveGameplayTag _newgameplaytags_2 _tag :!class /Script/GameplayTags.BlueprintGameplayTagLibrary)))
    (PhotoMode|AddPhoto (CallFunction|GetPhotoStorage (Variables|Default|GetLocalPlayerController)) _photometadata)
    (return _photometadata)))
