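; Photo mode selfie stick actor that attaches to a player pawn's right-hand mesh socket and positions a camera at its far end.
; Each tick grows the static mesh's Z scale toward a target length, then places the camera at the end point with fixed offsets, turned back along the stick with a yaw correction.
; Removal resets the camera's relative transform, hides the stick, stops ticking and lets it expire shortly after.
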(event EventTick
  (params (DeltaSeconds :type Float%20%28single-precision%29))
  (Utilities|FlowControl|Sequence
    (:then_0
      (Variables|Default|SetCurrentScale (Math|Float|Clamp(Float) (+ (Variables|Default|GetCurrentScale) (* DeltaSeconds (Variables|Default|GetScaleSpeed))) 0.0 (Variables|Default|GetTargetScale)))
      (CallFunction|UpdateScale))
    (:then_1
      (CallFunction|UpdateCameraTransform))))

(event Custom|InitializeStick
  (params (Camera :type Actor%20Object%20Reference) (RelativeRotation :type Rotator) (Player :type Fort%20Pawn%20Object%20Reference))
  (Variables|Default|SetCamera Camera)
  (Variables|Default|SetPlayer Player)
  (bind _output_get (Variables|Default|SetAttach%20to%20Mesh (Variables|/Script/Engine.Character|GetMesh Player)))
  (Transformation|AttachActorToComponent _output_get "hand_attach_r" "SnapToTarget" "KeepWorld" "KeepWorld")
  (Transformation|SetActorRelativeRotation RelativeRotation))

(event Custom|RemoveStick
  (Transformation|SetActorRelativeTransform (Variables|Default|GetCamera) (struct :Location "0, 0, 0" :Rotation "0, 0, 0" :Scale "1.000000,1.000000,1.000000") false false)
  (Actor|Tick|SetActorTickEnabled false)
  (Rendering|SetActorHiddenInGame true)
  (Actor|SetLifeSpan 0.1))

(fn GetCameraLocation
  (return (Utilities|Operators|Add (Transformation|GetWorldLocation (Variables|Default|GetEndLocation)) (* (Transformation|GetActorRightVector (Variables|Default|GetCamera)) 7.0) (* (Transformation|GetActorUpVector (Variables|Default|GetCamera)) 1e+01))))

(fn UpdateCameraTransform
  (bind _camera (Variables|Default|GetCamera)
    (:then
      (Transformation|SetActorLocation _camera (CallFunction|GetCameraLocation) false false)
      (Transformation|SetActorRotation (Variables|Default|GetCamera) (Math|Rotator|MakeRotfromX (* (Transformation|GetActorForwardVector) -1.0)) false)
      (Transformation|AddActorLocalRotation (Variables|Default|GetCamera) "0.000000,-90.000000,0.000000" false false))
    (:else)))

(fn UpdateScale
  (bind (_returnvalue_x _returnvalue_y) (Math|Vector|BreakVector (Transformation|GetWorldScale (Variables|Default|GetStaticMesh))))
  (Transformation|SetWorldScale3D (Variables|Default|GetStaticMesh) (Math|Vector|MakeVector _returnvalue_x _returnvalue_y (Variables|Default|GetCurrentScale))))

(fn ConstructionScript)
