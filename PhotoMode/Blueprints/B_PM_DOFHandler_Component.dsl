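; Photo camera component that drives autofocus depth of field while equipped, gated on Tick by the PhotoMode.Camera.DOF console variable.
; Target pawns, gathered from ability system components, are filtered by view cone and range, then insertion-sorted by screen-centre and proximity weight.
; The top target's distance becomes the focus target; focus and blur ease over FocusTime, with blur fading without targets.
; Results feed the player's post-process component and camera's scene capture.
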(event EventTick
  (params (DeltaSeconds :type Float%20%28single-precision%29))
  (bind _bissettingdof (Variables|Default|GetbIsSettingDOF))
  (bind _returnvalue (Development|GetConsoleVariableBoolValue "PhotoMode.Camera.DOF"))
  (if _returnvalue
    (if (Variables|/PhotoMode/Blueprints/B_PM_PhotoCamera.B_PM_PhotoCamera_C|GetbEquipped (stale EventBeginPlay _asb_pm_photo_camera))
      (if (not _bissettingdof)
        (CallFunction|EnableFOV))
      (bind _bfoundtarget (CallFunction|GetSortedDistances))
      (CallFunction|GetWeightedTargetDistance _bfoundtarget)
      (CallFunction|CalculateNewDistanceAndDOF _bfoundtarget)
      (CallFunction|UpdateDOF)
      (elif _bissettingdof
        (CallFunction|DisableFOV)))))

(event EventBeginPlay
  (bind _bissettingdof (Variables|Default|GetbIsSettingDOF))
  (bind _asb_pm_photo_camera (Utilities|Casting|CastToB_PM_PhotoCamera (Components|GetOwner)))
  (if (Variables|/PhotoMode/Blueprints/B_PM_PhotoCamera.B_PM_PhotoCamera_C|GetbEquipped _asb_pm_photo_camera)
    (if (not _bissettingdof)
      (CallFunction|EnableFOV))
    (bind _bfoundtarget (CallFunction|GetSortedDistances))
    (CallFunction|GetWeightedTargetDistance _bfoundtarget)
    (CallFunction|CalculateNewDistanceAndDOF _bfoundtarget)
    (CallFunction|UpdateDOF)
    (elif _bissettingdof
      (CallFunction|DisableFOV))))

(fn CalculateNewDistanceandDOF
  (params (bHasTarget :type Boolean))
  (bind _returnvalue (Utilities|Time|GetWorldDeltaSeconds))
  (bind _focustime (Variables|Default|GetFocusTime))
  (bind _returnvalue_2 (/ _returnvalue _focustime))
  (bind _returnvalue_3 (Math|Float|Min(Float) _returnvalue_2 1.0))
  (Variables|Default|SetCurrentDistance (+ (* (- (Variables|Default|GetTargetDistance) (Variables|Default|GetCurrentDistance)) _returnvalue_3) (Variables|Default|GetCurrentDistance)))
  (Variables|Default|SetCurrentDOF (+ (* (- (Math|Float|SelectFloat (Variables|Default|GetDOFAmount) 0.0 bHasTarget) (Variables|Default|GetCurrentDOF)) _returnvalue_3) (Variables|Default|GetCurrentDOF))))

(fn DisableFOV
  (Variables|Default|SetbIsSettingDOF false)
  (bind _returnvalue (CallFunction|GetPlayerPawn))
  (Utilities|IsValid (Actor|GetComponentbyClass _returnvalue "/Script/Engine.PostProcessComponent")
    (:Is%20Valid
      (Utilities|Struct|SetmembersinPostProcessSettings (Variables|/Script/Engine.PostProcessComponent|GetSettings (Actor|GetComponentbyClass _returnvalue "/Script/Engine.PostProcessComponent")) 24.576 0.0))))

(fn EnableFOV
  (Variables|Default|SetbIsSettingDOF true))

(fn GetSortedDistances
  (bind _sorteddistances (Variables|Default|GetSortedDistances))
  (bind _weights (LocalVariables|GetWeights))
  (bind (_location _direction _fov _fovdot) (CallFunction|GetPlayerCameraData))
  (Utilities|Array|Resize (Variables|Default|GetSortedDistances))
  (Utilities|Array|Resize (LocalVariables|GetWeights))
  (Utilities|FlowControl|Sequence
    (:then_0
      (for _array_element (Variables|Default|GetTargetActors)
        (if (and (> (Math|Vector|DotProduct (/ (- (Transformation|GetActorLocation _array_element) _location) (Math|Vector|VectorLength (- (Transformation|GetActorLocation _array_element) _location))) _direction) _fovdot) (< (Math|Vector|VectorLength (- (Transformation|GetActorLocation _array_element) _location)) (Variables|Default|GetMaxDistanceThreshold)))
          (bind _returnvalue (Utilities|Array|Add (Variables|Default|GetSortedDistances) (Math|Vector|VectorLength (- (Transformation|GetActorLocation _array_element) _location))))
          (Utilities|Array|Add (LocalVariables|GetWeights) (Math|Vector2D|MakeVector2D (* (Math|Float|Min(Float) 1.0 (Math|Float|Max(Float) (* (+ (Variables|Default|GetWeightedCenterRadius) 1.0) (/ (- (Math|Vector|DotProduct (/ (- (Transformation|GetActorLocation _array_element) _location) (Math|Vector|VectorLength (- (Transformation|GetActorLocation _array_element) _location))) _direction) _fovdot) (- 1.0 _fovdot))))) (/ 1e+02 (Math|Float|Max(Float) (Math|Vector|VectorLength (- (Transformation|GetActorLocation _array_element) _location)) 0.001))) (Math|Conversions|ToFloat(Integer) _returnvalue))))))
    (:then_1
      (if (== (Utilities|Array|Length _weights) 0)
        (return false)))
    (:then_2
      (CallFunction|InsertionSortVector2 _weights))
    (:then_3
      (if (> (Utilities|Array|Length _sorteddistances) 1)
        (bind _output_get (LocalVariables|SetTempSortedDistances _sorteddistances))
        (for (_array_element_2 _array_index) _weights
          (Utilities|Array|SetArrayElem (Variables|Default|GetSortedDistances) _array_index (Utilities|Array|Get(acopy) _output_get (Math|Float|Truncate (.Y (Math|Vector2D|BreakVector2D _array_element_2))))))))
    (:then_4
      (return true))))

(fn GetWeightedTargetDistance
  (params (bHasNewTarget :type Boolean))
  (if bHasNewTarget
    (Variables|Default|SetTargetDistance (Math|Float|SelectFloat (Utilities|Array|Get(acopy) (Variables|Default|GetSortedDistances)) 0.0 (> (Utilities|Array|Length (Variables|Default|GetSortedDistances)) 0)))))

(fn GetNewTargets
  (params (TargetAbilityComponents :type Array%20of%20Ability%20System%20Component%20Object%20References))
  (bind _returnvalue (Utilities|Array|Length TargetAbilityComponents))
  (Utilities|Array|Resize (Variables|Default|GetSortedDistances) _returnvalue)
  (Utilities|Array|Resize (Variables|Default|GetTargetPositions) _returnvalue)
  (Utilities|Array|Resize (Variables|Default|GetTargetActors) _returnvalue)
  (for (_array_element _array_index) TargetAbilityComponents
    (bind _asplayer_state (Utilities|Casting|CastToPlayerState (Components|GetOwner _array_element)))
    (Utilities|Array|SetArrayElem (Variables|Default|GetTargetActors) _array_index (PlayerState|GetPawn _asplayer_state))))

(fn GetPlayerCameraData
  (return (Camera|GetCameraLocation (Game|GetPlayerCameraManager 0)) (Math|Vector|GetForwardVector (Camera|GetCameraRotation (Game|GetPlayerCameraManager 0))) (Camera|GetFOVAngle (Game|GetPlayerCameraManager 0)) (Math|Trig|Cos(Degrees) (/ (Camera|GetFOVAngle (Game|GetPlayerCameraManager 0)) 2.0))))

(fn GetPlayerPawn
  (return (Actor|GetOwner (Components|GetOwner))))

(fn InsertionSortVector2
  (params (Vector2DArray :type Array%20of%20Vector%202D%20Structures))
  (if (> (Utilities|Array|Length Vector2DArray) 1)
    (for _index (range 1 (Utilities|Array|Length Vector2DArray))
      (bind _output_get (LocalVariables|Setj _index))
      (while (and (> (Math|Vector2D|BreakVector2D (Utilities|Array|Get(acopy) Vector2DArray _output_get)) (Math|Vector2D|BreakVector2D (Utilities|Array|Get(acopy) Vector2DArray (- _output_get 1)))) (> _output_get 0))
        (bind _output_get_2 (LocalVariables|Setk (Utilities|Array|Get(acopy) Vector2DArray _output_get)))
        (Utilities|Array|SetArrayElem Vector2DArray _output_get (Utilities|Array|Get(acopy) Vector2DArray (- _output_get 1)))
        (Utilities|Array|SetArrayElem Vector2DArray (- _output_get 1) _output_get_2)
        (Math|Integer|DecrementInt (LocalVariables|Getj))))))

(fn UpdateDOF
  (bind _returnvalue (Components|GetOwner))
  (bind _returnvalue_2 (Actor|GetComponentbyClass _returnvalue "/Script/Engine.SceneCaptureComponent2D"))
  (Utilities|IsValid (Variables|Default|GetCamPP)
    (:Is%20Valid
      _)
    (:Is%20Not%20Valid
      (bind _returnvalue_3 (CallFunction|GetPlayerPawn))
      (bind _output_get (Variables|Default|SetCamPP (Actor|GetComponentbyClass _returnvalue_3 "/Script/Engine.PostProcessComponent")))
      (Utilities|IsValid _output_get
        (:Is%20Valid
          (bind _structout (Utilities|Struct|SetmembersinPostProcessSettings (Variables|/Script/Engine.PostProcessComponent|GetSettings _output_get) (Variables|Default|GetCurrentDOF) (Variables|Default|GetCurrentDistance))))
        (:Is%20Not%20Valid
          _))
      (Utilities|IsValid _returnvalue_2
        (:Is%20Valid
          (Variables|/Script/Engine.SceneCaptureComponent2D|SetPostProcessSettings (Utilities|Copy _structout) _returnvalue_2)
          (Utilities|Struct|SetmembersinPostProcessSettings (Variables|/Script/Engine.SceneCaptureComponent2D|GetPostProcessSettings _returnvalue_2) (Variables|Default|GetCurrentDOF) (Variables|Default|GetCurrentDistance)))))))
