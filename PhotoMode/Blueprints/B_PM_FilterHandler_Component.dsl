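; Photo Mode component on B_PM_PhotoCamera that applies camera filters as post-process materials on the owning player pawn.
; Filter-changed UI messages resolve a tag to a PhotoFilterAsset and blocking-load its material; camera equip and unequip enable or disable the filter.
; The blendable's array index is cached and searched for again when stale.
; A material parameter collection "Enabled" scalar turns off on photo capture and back on at the photo-taken message.
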(event EventBeginPlay
  (bind _asb_pm_photo_camera (Utilities|Casting|CastToB_PM_PhotoCamera (Components|GetOwner)))
  (Default|BindEventtoOnPhotoCapture _asb_pm_photo_camera (delegate Custom|On%20Photo%20Capture))
  (Default|BindEventtoOnCameraEquipped _asb_pm_photo_camera (delegate Custom|OnEquip))
  (Default|BindEventtoOnCameraUnequipped _asb_pm_photo_camera (delegate Custom|OnUnequip))
  (bind (_asynctaskproxy _payload) (GameplayEventRouter|StartListeningToEvent (GameplayEventRouter|GetGlobalEventRouter) "/PhotoMode/Blueprints/S_PM_PhotoCameraFilterChangedUIMessage.S_PM_PhotoCameraFilterChangedUIMessage" (Fortnite|GetFirstLocalFortPlayerController true))
    (:then
      (GameplayEventRouter|StartListeningToEvent (GameplayEventRouter|GetGlobalEventRouter) "/PhotoMode/Blueprints/S_PM_PhotoCameraPhotoTakenUIMessage.S_PM_PhotoCameraPhotoTakenUIMessage" (Fortnite|GetFirstLocalFortPlayerController true)
        (:OnEventReceived
          (CallFunction|ToggleFilterinMaterial true))))
    (:OnEventReceived
      (CallFunction|SetNewFilterfromTag (Utilities|Struct|BreakSPMPhotoCameraFilterChangedUIMessage _payload)))))

(event Custom|OnEquip
  (CallFunction|EnableorUpdateFilter 0))

(event Custom|On%20Photo%20Capture
  (CallFunction|ToggleFilterinMaterial false))

(event Custom|OnUnequip
  (CallFunction|DisableFilter))

(fn DisableFilter
  (bind _matppindex (Variables|Default|GetMatPPIndex))
  (bind _returnvalue (CallFunction|GetPlayerPawn))
  (Utilities|IsValid (Actor|GetComponentbyClass _returnvalue "/Script/Engine.PostProcessComponent")
    (:Is%20Valid
      (if (>= _matppindex 0)
        (if (not (== (.Object (Utilities|Struct|BreakWeightedBlendable (Utilities|Array|Get(aref) (Utilities|Struct|BreakWeightedBlendables (Utilities|Struct|BreakPostProcessSettings (Variables|/Script/Engine.PostProcessComponent|GetSettings (Actor|GetComponentbyClass _returnvalue "/Script/Engine.PostProcessComponent")))) _matppindex))) (Variables|Default|GetCurrentLUTMat)))
          (bind _bsuccess (CallFunction|FindMaterialonPP (Utilities|Struct|BreakWeightedBlendables (Utilities|Struct|BreakPostProcessSettings (Variables|/Script/Engine.PostProcessComponent|GetSettings (Actor|GetComponentbyClass _returnvalue "/Script/Engine.PostProcessComponent"))))))
          (if _bsuccess
            _))
        (Utilities|Struct|SetmembersinWeightedBlendable (Utilities|Array|Get(aref) (Utilities|Struct|BreakWeightedBlendables (Utilities|Struct|BreakPostProcessSettings (Variables|/Script/Engine.PostProcessComponent|GetSettings (Actor|GetComponentbyClass _returnvalue "/Script/Engine.PostProcessComponent")))) _matppindex) 0.0)))))

(fn EnableOrUpdateFilter
  (params (NewPPMaterial :type Material%20Interface%20Object%20Reference))
  (bind _returnvalue (Utilities|IsValid NewPPMaterial :!class /Script/Engine.KismetSystemLibrary))
  (bind _weightedblendable (Utilities|Struct|MakeWeightedBlendable 1.0 (select _returnvalue NewPPMaterial (Variables|Default|GetCurrentLUTMat))))
  (bind _currentlutmat (Variables|Default|GetCurrentLUTMat))
  (bind _returnvalue_2 (select _returnvalue NewPPMaterial _currentlutmat))
  (bind _matppindex (Variables|Default|GetMatPPIndex))
  (bind _returnvalue_3 (CallFunction|GetPlayerPawn))
  (Utilities|IsValid (Actor|GetComponentbyClass _returnvalue_3 "/Script/Engine.PostProcessComponent")
    (:Is%20Valid
      (if (>= _matppindex 0)
        (if (not (== (.Object (Utilities|Struct|BreakWeightedBlendable (Utilities|Array|Get(aref) (Utilities|Struct|BreakWeightedBlendables (Utilities|Struct|BreakPostProcessSettings (Variables|/Script/Engine.PostProcessComponent|GetSettings (Actor|GetComponentbyClass _returnvalue_3 "/Script/Engine.PostProcessComponent")))) _matppindex))) (Variables|Default|GetCurrentLUTMat)))
          (bind _bsuccess (CallFunction|FindMaterialonPP (Utilities|Struct|BreakWeightedBlendables (Utilities|Struct|BreakPostProcessSettings (Variables|/Script/Engine.PostProcessComponent|GetSettings (Actor|GetComponentbyClass _returnvalue_3 "/Script/Engine.PostProcessComponent"))))))
          (if _bsuccess
            (Utilities|Array|SetArrayElem (Utilities|Struct|BreakWeightedBlendables (Utilities|Struct|BreakPostProcessSettings (Variables|/Script/Engine.PostProcessComponent|GetSettings (Actor|GetComponentbyClass _returnvalue_3 "/Script/Engine.PostProcessComponent")))) _matppindex _weightedblendable))
          (Variables|Default|SetCurrentLUTMat _returnvalue_2))
        (else
          (bind _returnvalue_4 (Utilities|Array|Add (Utilities|Struct|BreakWeightedBlendables (Utilities|Struct|BreakPostProcessSettings (Variables|/Script/Engine.PostProcessComponent|GetSettings (Actor|GetComponentbyClass _returnvalue_3 "/Script/Engine.PostProcessComponent")))) _weightedblendable))
          (Variables|Default|SetMatPPIndex _returnvalue_4))))))

(fn FindMaterialOnPP
  (params (InputPin :type Array%20of%20Weighted%20Blendable%20Structures))
  (Variables|Default|SetMatPPIndex -1)
  (for (_array_element _array_index) InputPin
    (if (== (.Object (Utilities|Struct|BreakWeightedBlendable _array_element)) (Variables|Default|GetCurrentLUTMat))
      (Variables|Default|SetMatPPIndex _array_index)))
  (return (>= (Variables|Default|GetMatPPIndex) 0)))

(fn GetPlayerPawn
  (return (Actor|GetOwner (Components|GetOwner))))

(fn SetNewFilterFromTag
  (params (Filter%20Tag :type Gameplay%20Tag%20Structure))
  (if (== Filter%20Tag 0)
    (CallFunction|DisableFilter)
    (else
      (bind _returnvalue (Utilities|LoadAssetBlocking (Variables|/Script/PhotoModeRuntime.PhotoFilterAsset|GetPPFXMaterial (Class|PhotoFilterAsset|GetFilterAssetwithTag Filter%20Tag))))
      (bind _asmaterial_interface (Utilities|Casting|CastToMaterialInterface _returnvalue))
      (CallFunction|EnableorUpdateFilter _asmaterial_interface))))

(fn ToggleFilterInMaterial
  (params (bEnable :type Boolean))
  (Rendering|Material|SetScalarParameterValue (Variables|Default|GetFilterParameterCollection) (Utilities|Name|MakeLiteralName "Enabled") (Math|Float|SelectFloat 1.0 0.0 bEnable) :!class /Script/Engine.KismetMaterialLibrary))
