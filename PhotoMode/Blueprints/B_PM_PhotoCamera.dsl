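; Photo Mode camera weapon that players equip to take photos, zoom, switch facing and orientation, wave, emote and open the album.
; On equip it links a camera anim layer, grants camera gameplay abilities on the server, activates aim and caches the previous weapon; unequip reverses this and re-equips that weapon.
; A max-priority input mapping context follows equip and holster state.
; Helper functions activate abilities by spec handle, with a timer enforcing minimum zoom duration.
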(event Custom|OnHolsterChanged
  (CallFunction|UpdateInputs (Variables|Default|GetCachedController)))

(event EventEndPlay
  (params (EndPlayReason :type EEndPlayReason%20Enum :default "Destroyed"))
  (bind _returnvalue (Game|GetInstigator))
  (bind _outputdelegate (EventDispatchers|CreateEvent :!event Custom|OnHolsterChanged))
  ;[1 0.3378 0.01649 0.5] EndPlay
  (|CallParentFunction EndPlayReason)
  (Variables|Default|SetbEquipped false)
  (Utilities|IsValid _returnvalue
    (:Is%20Valid
      (bind _asfort_pawn (Utilities|Casting|CastToFortPawn _returnvalue))
      (EventDispatchers|UnbindEventfromOnHolsteredEvent _asfort_pawn _outputdelegate)
      (EventDispatchers|UnbindEventfromOnUnholsteredEvent _asfort_pawn _outputdelegate))))

(event FortWeapon|EventOnEquippedAndAttached
  (bind _outputdelegate (EventDispatchers|CreateEvent :!event Custom|OnHolsterChanged))
  ;[0 0.01216 1 0.5] On Equip
  ;; Cache Previous Weapon
  (bind _asfort_player_pawn (Utilities|Casting|CastToFortPlayerPawn (Game|GetInstigator))
    (:then
      (FortPawn|BindEventtoOnWeaponEquippedDelegate _asfort_player_pawn (delegate Custom|CustomEventOnWeaponEquippedEvent)))
    (:CastFailed
      _))
  (Default|CallOnCameraEquipped)
  (CallFunction|LinkAnimLayer)
  (CallFunction|GetAbilitySystem)
  (Variables|Default|SetbEquipped true)
  (CallFunction|GiveCamAbilities)
  (bind _asfort_pawn (Utilities|Casting|CastToFortPawn (Game|GetInstigator)))
  (if (Pawn|IsLocallyControlled _asfort_pawn)
    (EventDispatchers|BindEventtoOnHolsteredEvent _asfort_pawn _outputdelegate)
    (EventDispatchers|AssignOnUnholsteredEvent _asfort_pawn _outputdelegate)
    (bind _output_get (Variables|Default|SetCachedController (Fortnite|GetFortPlayerControllerfromActor _asfort_pawn)))
    (CallFunction|UpdateInputs _output_get)))

(event Custom|CustomEventOnWeaponEquippedEvent
  (params (NewWeapon :type Fort%20Weapon%20Object%20Reference) (PrevWeapon :type Fort%20Weapon%20Object%20Reference))
  ;[0 0.01216 1 0.5] On Equip
  ;; Cache Previous Weapon
  (Variables|Default|SetCachedWeapon PrevWeapon))

(event FortWeapon|EventOnUnEquippedAndDetached
  (bind _outputdelegate (EventDispatchers|CreateEvent :!event Custom|OnHolsterChanged))
  (bind _cachedweapon (Variables|Default|GetCachedWeapon))
  ;[0.1458 0.05437 0 0.5] On Unequip
  (Default|CallOnCameraUnequipped)
  (Default|UnlinkAnimLayer)
  (Variables|Default|SetbEquipped false)
  (CallFunction|RemoveCamAbilities)
  (bind _asfort_pawn (Utilities|Casting|CastToFortPawn (Game|GetInstigator)))
  (EventDispatchers|UnbindEventfromOnHolsteredEvent _asfort_pawn _outputdelegate)
  (EventDispatchers|UnbindEventfromOnUnholsteredEvent _asfort_pawn _outputdelegate)
  (CallFunction|UpdateInputs (Variables|Default|GetCachedController))
  ;; Equip Cached Weapon
  (bind _asfort_pawn_2 (Utilities|Casting|CastToFortPawn (Game|GetInstigator)))
  (Utilities|IsValid _cachedweapon
    (:Is%20Valid
      (Weapon|EquipWeaponActor _asfort_pawn_2 _cachedweapon))))

(event Custom|UpdateInputs
  (params (PlayerController :type Player%20Controller%20Object%20Reference))
  (bind _returnvalue (PlayerController|LocalPlayerSubsystems|GetEnhancedInputLocalPlayerSubsystem PlayerController))
  (bind _cameramappingcontext (Variables|Default|GetCameraMappingContext))
  ;[0.000838 0.7657 0.9948 0.5] UpdateInputs
  (Networking|CanExecuteCosmeticEvents
    (:True
      (Utilities|IsValid PlayerController
        (:Is%20Valid
          (if (and (Variables|Default|GetbEquipped) (not (CallFunction|IsHolstered)))
            ;[1 0 0.01763] TODO: Change Priority
            ;
            ; Set to a high number so camera takes priority, but this needs to be lowered (original 5)
            (Input|AddMappingContext _returnvalue _cameramappingcontext 2147483646 "(bIgnoreAllPressedKeysUntilRelease=True,bForceImmediately=False,bNotifyUserSettings=False)")
            (else
              (Input|RemoveMappingContext _returnvalue _cameramappingcontext "(bIgnoreAllPressedKeysUntilRelease=True,bForceImmediately=False,bNotifyUserSettings=False)"))))))))

(event EnhancedInputActionIA_PhotoMode_CapturePhoto_Gamepad
  (params (ActionValue :type Boolean) (ElapsedSeconds :type Float%20%28double-precision%29) (TriggeredSeconds :type Float%20%28double-precision%29) (InputAction :type Input%20Action%20Object%20Reference :default "/PhotoMode/Input/Actions/IA_PhotoMode_CapturePhoto_Gamepad.IA_PhotoMode_CapturePhoto_Gamepad")))

(event EnhancedInputActionIA_PhotoMode_CapturePhoto
  (params (ActionValue :type Boolean) (ElapsedSeconds :type Float%20%28double-precision%29) (TriggeredSeconds :type Float%20%28double-precision%29) (InputAction :type Input%20Action%20Object%20Reference :default "/PhotoMode/Input/Actions/IA_PhotoMode_CapturePhoto.IA_PhotoMode_CapturePhoto")))

(event EnhancedInputActionIA_PhotoMode_AlbumQuickAccess
  (params (ActionValue :type Boolean) (ElapsedSeconds :type Float%20%28double-precision%29) (TriggeredSeconds :type Float%20%28double-precision%29) (InputAction :type Input%20Action%20Object%20Reference :default "/PhotoMode/Input/Actions/IA_PhotoMode_AlbumQuickAccess.IA_PhotoMode_AlbumQuickAccess")))

(event EnhancedInputActionIA_PhotoMode_Zoom_In
  (params (ActionValue :type Boolean) (ElapsedSeconds :type Float%20%28double-precision%29) (TriggeredSeconds :type Float%20%28double-precision%29) (InputAction :type Input%20Action%20Object%20Reference :default "/PhotoMode/Input/Actions/IA_PhotoMode_Zoom_In.IA_PhotoMode_Zoom_In")))

(event Custom|CheckZoomToFinish
  (for _array_element (Variables|Default|GetPendingToFinishZoom)
    (CallFunction|UpdateZoom false _array_element)))

(event EnhancedInputActionIA_PhotoMode_Zoom_Out
  (params (ActionValue :type Boolean) (ElapsedSeconds :type Float%20%28double-precision%29) (TriggeredSeconds :type Float%20%28double-precision%29) (InputAction :type Input%20Action%20Object%20Reference :default "/PhotoMode/Input/Actions/IA_PhotoMode_Zoom_Out.IA_PhotoMode_Zoom_Out")))

(event EnhancedInputActionIA_PhotoMode_EmoteWheel
  (params (ActionValue :type Boolean) (ElapsedSeconds :type Float%20%28double-precision%29) (TriggeredSeconds :type Float%20%28double-precision%29) (InputAction :type Input%20Action%20Object%20Reference :default "/PhotoMode/Input/Actions/IA_PhotoMode_EmoteWheel.IA_PhotoMode_EmoteWheel")))

(event EnhancedInputActionIA_PhotoMode_NextOrientation
  (params (ActionValue :type Boolean) (ElapsedSeconds :type Float%20%28double-precision%29) (TriggeredSeconds :type Float%20%28double-precision%29) (InputAction :type Input%20Action%20Object%20Reference :default "/PhotoMode/Input/Actions/IA_PhotoMode_NextOrientation.IA_PhotoMode_NextOrientation")))

(event EnhancedInputActionIA_PhotoMode_ToggleFacing
  (params (ActionValue :type Boolean) (ElapsedSeconds :type Float%20%28double-precision%29) (TriggeredSeconds :type Float%20%28double-precision%29) (InputAction :type Input%20Action%20Object%20Reference :default "/PhotoMode/Input/Actions/IA_PhotoMode_ToggleFacing.IA_PhotoMode_ToggleFacing")))

(event EnhancedInputActionIA_PhotoMode_Wave
  (params (ActionValue :type Boolean) (ElapsedSeconds :type Float%20%28double-precision%29) (TriggeredSeconds :type Float%20%28double-precision%29) (InputAction :type Input%20Action%20Object%20Reference :default "/PhotoMode/Input/Actions/IA_PhotoMode_Wave.IA_PhotoMode_Wave")))

(event EnhancedInputActionIA_PhotoMode_CallOver
  (params (ActionValue :type Boolean) (ElapsedSeconds :type Float%20%28double-precision%29) (TriggeredSeconds :type Float%20%28double-precision%29) (InputAction :type Input%20Action%20Object%20Reference :default "/PhotoMode/Input/Actions/IA_PhotoMode_CallOver.IA_PhotoMode_CallOver")))

(event Custom|EndZoom
  (bind (_bisinstance _returnvalue) (Ability|GameplayAbility|GetGameplayAbilityfromSpecHandle (Variables|Default|GetAbilitySystem) (Variables|Default|GetCameraZoomOutAbility)))
  (bind (_bisinstance_2 _returnvalue_2) (Ability|GameplayAbility|GetGameplayAbilityfromSpecHandle (Variables|Default|GetAbilitySystem) (Variables|Default|GetCameraZoomInAbility)))
  (Ability|EndAbilityLocally _returnvalue)
  (Ability|EndAbilityLocally _returnvalue_2))

(fn CallOver
  (bind _abilitysystem (CallFunction|GetAbilitySystem))
  (Abilities|TryActivateAbility _abilitysystem (Variables|Default|GetCallOverAbility) true))

(fn CanApplyHolster
  (params (HolsterId :type Name%20%28by%20ref%29 :default "None"))
  (return true))

(fn CapturePhoto
  (Default|CallOnPhotoCapture)
  (bind _abilitysystem (CallFunction|GetAbilitySystem))
  (Abilities|TryActivateAbility _abilitysystem (Variables|Default|GetCapturePhotoAbility) true))

(fn GetAbilitySystem
  (Utilities|IsValid (Variables|Default|GetAbilitySystem)
    (:Is%20Valid
      _)
    (:Is%20Not%20Valid
      (Variables|Default|SetAbilitySystem (Ability|GetAbilitySystemComponent (Game|GetInstigator)))))
  (return (Variables|Default|GetAbilitySystem)))

(fn GetActorAudioParameters
  (return 0))

(fn GetAffiliationsToShowFor
  (return "Friendly"))

(fn GiveAbilityToInstigator
  (params (Ability :type Gameplay%20Ability%20Class%20Reference))
  (if (and (Networking|IsServer) (!= Ability 0))
    (bind _abilitysystem (CallFunction|GetAbilitySystem))
    (bind _returnvalue (GameplayAbilities|GiveAbility _abilitysystem Ability 0 -1))
    (return _returnvalue)
    (else
      (return 0))))

(fn GiveCamAbilities
  (Networking|SwitchHasAuthority
    (:Authority
      (Utilities|FlowControl|Sequence
        (:then_0
          (bind _givenability (Default|GiveAbilitytoInstigator "/PhotoMode/Abilities/GA_PM_Camera_EmoteWheel.GA_PM_Camera_EmoteWheel_C"))
          (Variables|Default|SetEmoteWheelAbility _givenability))
        (:then_1
          (bind _givenability_2 (Default|GiveAbilitytoInstigator "/PhotoMode/Abilities/GA_PM_Camera_ToggleFacing.GA_PM_Camera_ToggleFacing_C"))
          (Variables|Default|SetToggleFacingAbility _givenability_2))
        (:then_2
          (bind _givenability_3 (Default|GiveAbilitytoInstigator "/PhotoMode/Abilities/GA_PM_Camera_Orientation_Next.GA_PM_Camera_Orientation_Next_C"))
          (Variables|Default|SetNextOrientationAbility _givenability_3))
        (:then_3
          (bind _givenability_4 (Default|GiveAbilitytoInstigator "/PhotoMode/Abilities/GA_PM_Camera_Orientation_Prev.GA_PM_Camera_Orientation_Prev_C"))
          (Variables|Default|SetPrevOrientationAbillity _givenability_4))
        (:then_4
          (bind _givenability_5 (Default|GiveAbilitytoInstigator "/PhotoMode/Abilities/GA_PM_Camera_ToggleUI.GA_PM_Camera_ToggleUI_C"))
          (Variables|Default|SetToggleUIAbility _givenability_5))
        (:then_5
          (bind _givenability_6 (Default|GiveAbilitytoInstigator "/PhotoMode/Abilities/GA_PM_Camera_ZoomIn.GA_PM_Camera_ZoomIn_C"))
          (Variables|Default|SetCameraZoomInAbility _givenability_6))
        (:then_6
          (bind _givenability_7 (Default|GiveAbilitytoInstigator "/PhotoMode/Abilities/GA_PM_Camera_ZoomOut.GA_PM_Camera_ZoomOut_C"))
          (Variables|Default|SetCameraZoomOutAbility _givenability_7))
        (:then_7
          (bind _givenability_8 (Default|GiveAbilitytoInstigator "/PhotoMode/Abilities/GA_PM_Camera_OpenAlbum.GA_PM_Camera_OpenAlbum_C"))
          (Variables|Default|SetOpenAlbumAbility _givenability_8))
        (:then_8
          (bind _givenability_9 (Default|GiveAbilitytoInstigator "/PhotoMode/Abilities/GA_PM_Camera_TakePhotograph.GA_PM_Camera_TakePhotograph_C"))
          (Variables|Default|SetCapturePhotoAbility _givenability_9))
        (:then_9
          (bind _givenability_10 (Default|GiveAbilitytoInstigator (Variables|Ability|Class|GetAimAbilityClass)))
          (bind _output_get (Variables|Default|SetCameraAimAbility _givenability_10))
          (bind _abilitysystem (CallFunction|GetAbilitySystem))
          (Abilities|TryActivateAbility _abilitysystem _output_get true))
        (:then_10
          (bind _givenability_11 (Default|GiveAbilitytoInstigator (Variables|Ability|Class|GetSelfieAbilityClass)))
          (Variables|Default|SetSelfieFacingAbility _givenability_11))
        (:then_11
          (bind _givenability_12 (Default|GiveAbilitytoInstigator "/PhotoMode/Abilities/GA_PM_Camera_BackFacing.GA_PM_Camera_BackFacing_C"))
          (Variables|Default|SetBackFacingAbility _givenability_12))
        (:then_12
          (bind _givenability_13 (Default|GiveAbilitytoInstigator "/PhotoMode/Abilities/GA_PM_Camera_TakePhotograph_Hipfire.GA_PM_Camera_TakePhotograph_Hipfire_C"))
          (Variables|Default|SetCapturePhotoHipfireAbility _givenability_13))
        (:then_13
          (bind _givenability_14 (Default|GiveAbilitytoInstigator "/PhotoMode/Abilities/GA_PM_Camera_Wave.GA_PM_Camera_Wave_C"))
          (Variables|Default|SetWaveAbility _givenability_14))
        (:then_14
          (if (Utilities|IsValidClass 0)
            (bind _givenability_15 (Default|GiveAbilitytoInstigator (Variables|Ability|Class|GetCallOverAbilityClass)))
            (Variables|Default|SetCallOverAbility _givenability_15)))
        (:then_15
          (bind _givenability_16 (Default|GiveAbilitytoInstigator "/PhotoMode/Abilities/GA_PM_Camera_Emote.GA_PM_Camera_Emote_C"))
          (Variables|Default|SetEmoteAbility _givenability_16))
        (:then_16
          (bind _givenability_17 (Default|GiveAbilitytoInstigator "/PhotoMode/Abilities/GA_PM_Camera_EmoteEnd.GA_PM_Camera_EmoteEnd_C"))
          (Variables|Default|SetEmoteEndAbility _givenability_17))))))

(fn IsHolstered
  (bind _asfort_pawn (Utilities|Casting|CastToFortPawn (Game|GetInstigator))
    (:then
      (return (Weapon|IsWeaponHolstered _asfort_pawn)))
    (:CastFailed
      (return false))))

(fn LinkAnimLayer
  (if (Utilities|IsValidClass (Variables|Default|GetAnimLayer_Camera))
    (bind _asfort_pawn (Utilities|Casting|CastToFortPawn (Game|GetInstigator)))
    (Animation|LinkedAnimGraphs|LinkAnimClassLayers (Components|SkeletalMesh|GetAnimInstance (Variables|/Script/Engine.Character|GetMesh _asfort_pawn)) (Variables|Default|GetAnimLayer_Camera))
    (else
      (Development|PrintString "ERROR: Error in B_PM_PhotoCamera. Tried to Link Anim Class layer but Anim Layer Camera is empty! Please add an Anim Layer for the camera! Aborting!" true true "(R=1.000000,G=0.017908,B=0.000000,A=1.000000)"))))

(fn NextOrientation
  (bind _abilitysystem (CallFunction|GetAbilitySystem))
  (Abilities|TryActivateAbility _abilitysystem (Variables|Default|GetNextOrientationAbility) true))

(fn OnRep_TargetedNPCs
  (Class|BPMDOFHandlerComponent|GetNewTargets (Variables|Default|GetB_PM_DOFHandler_Component) (Variables|Default|GetTargetedNPCs)))

(fn PrevOrientation
  (bind _abilitysystem (CallFunction|GetAbilitySystem))
  (Abilities|TryActivateAbility _abilitysystem (Variables|Default|GetPrevOrientationAbillity) true))

(fn QuickAccessAlbum
  (bind _abilitysystem (CallFunction|GetAbilitySystem))
  (Abilities|TryActivateAbility _abilitysystem (Variables|Default|GetOpenAlbumAbility) true))

(fn RemoveCamAbilities
  (Networking|SwitchHasAuthority
    (:Authority
      (for _array_element (Utilities|Array|MakeArray (Variables|Default|GetEmoteWheelAbility) (Variables|Default|GetOpenAlbumAbility) (Variables|Default|GetCameraZoomInAbility) (Variables|Default|GetToggleFacingAbility) (Variables|Default|GetPrevOrientationAbillity) (Variables|Default|GetNextOrientationAbility) (Variables|Default|GetToggleUIAbility) (Variables|Default|GetCapturePhotoAbility) (Variables|Default|GetCameraZoomOutAbility) (Variables|Default|GetCameraAimAbility) (Variables|Default|GetCapturePhotoHipfireAbility) (Variables|Default|GetSelfieFacingAbility) (Variables|Default|GetBackFacingAbility) (Variables|Default|GetWaveAbility) (Variables|Default|GetEmoteAbility) (Variables|Default|GetEmoteEndAbility) (Variables|Default|GetCallOverAbility))
        (GameplayAbilities|ClearAbility (Variables|Default|GetAbilitySystem) _array_element)))))

(fn ShouldShowSoundIndicator
  (params (PlayerController :type Fort%20Player%20Controller%20Object%20Reference))
  (return false))

(fn ShowEmoteWheel
  (bind _abilitysystem (CallFunction|GetAbilitySystem))
  (Abilities|TryActivateAbility _abilitysystem (Variables|Default|GetEmoteWheelAbility) true))

(fn ToggleFacing
  (bind _abilitysystem (CallFunction|GetAbilitySystem))
  (Abilities|TryActivateAbility _abilitysystem (Variables|Default|GetToggleFacingAbility) true))

(fn ToggleUI
  (bind _abilitysystem (CallFunction|GetAbilitySystem))
  (Abilities|TryActivateAbility _abilitysystem (Variables|Default|GetToggleUIAbility) true))

(fn TryActivateAbilitiesByTag
  (params (TagContainer :type Gameplay%20Tag%20Container%20Structure) (bAllowRemoteActivation :type Boolean))
  (bind _abilitysystem (CallFunction|GetAbilitySystem))
  (bind _returnvalue (Abilities|TryActivateAbilitiesbyTag _abilitysystem TagContainer bAllowRemoteActivation))
  (return _returnvalue))

(fn UnlinkAnimLayer
  (if (Utilities|IsValidClass (Variables|Default|GetAnimLayer_Camera))
    (bind _asfort_pawn (Utilities|Casting|CastToFortPawn (Game|GetInstigator)))
    (Animation|LinkedAnimGraphs|UnlinkAnimClassLayers (Components|SkeletalMesh|GetAnimInstance (Variables|/Script/Engine.Character|GetMesh _asfort_pawn)) (Variables|Default|GetAnimLayer_Camera))
    (else
      (Development|PrintString "ERROR: Error in B_PM_PhotoCamera. Tried to Unlink Anim Class layer but Anim Layer Camera is empty! Please add an Anim Layer for the camera! Aborting!" true true "(R=1.000000,G=0.017908,B=0.000000,A=1.000000)"))))

(fn UpdateZoom
  (params (bHasStartedAction :type Boolean) (ZoomAbilitySpec :type Gameplay%20Ability%20Spec%20Handle%20Structure))
  (bind _returnvalue (Utilities|Time|GetTimeSeconds))
  (bind _zoomabilityspec (Variables|GetZoomAbilitySpec))
  (bind _zoomtimes (Variables|Default|GetZoomTimes))
  (bind _value (Utilities|Map|Find _zoomtimes _zoomabilityspec))
  (bind _returnvalue_2 (- _returnvalue _value))
  (bind _minzoomtime (Variables|Default|GetMinZoomTime))
  (bind _abilitysystem (CallFunction|GetAbilitySystem))
  (if bHasStartedAction
    (Abilities|TryActivateAbility _abilitysystem (Variables|GetZoomAbilitySpec) false)
    (Utilities|Array|RemoveItem (Variables|Default|GetPendingToFinishZoom) (Variables|GetZoomAbilitySpec))
    (Utilities|Map|Add (Variables|Default|GetZoomTimes) (Variables|GetZoomAbilitySpec) (Utilities|Time|GetTimeSeconds))
    (elif (> _returnvalue_2 _minzoomtime)
      (Ability|EndAbilityLocally (.ReturnValue (Ability|GameplayAbility|GetGameplayAbilityfromSpecHandle _abilitysystem (Variables|GetZoomAbilitySpec))))
      (else
        (bind _returnvalue_3 (Utilities|Time|SetTimerbyEvent (EventDispatchers|CreateEvent :!event CheckZoomToFinish) (- _minzoomtime _returnvalue_2)))
        (Variables|Default|SetPendingZoomToFinishHandle _returnvalue_3)
        (Utilities|Array|Add (Variables|Default|GetPendingToFinishZoom) (Variables|GetZoomAbilitySpec))))))

(fn ConstructionScript
  (|CallParentFunction))

(fn Wave
  (bind _abilitysystem (CallFunction|GetAbilitySystem))
  (Abilities|TryActivateAbility _abilitysystem (Variables|Default|GetWaveAbility) true)
  (bind _returnvalue (Abilities|TryActivateAbility _abilitysystem (Variables|Default|GetEmoteAbility) true))
  (if (not _returnvalue)
    (Abilities|TryActivateAbility _abilitysystem (Variables|Default|GetEmoteEndAbility) true)))

(fn ZoomIn
  (params (bHasStartedAction :type Boolean))
  (CallFunction|UpdateZoom bHasStartedAction (Variables|Default|GetCameraZoomInAbility)))

(fn ZoomOut
  (params (bHasStartedAction :type Boolean))
  (CallFunction|UpdateZoom bHasStartedAction (Variables|Default|GetCameraZoomOutAbility)))
