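; Photo mode component on the player pawn that supplies facing-specific animation data and emote categories for Selfie and Back camera views.
; The owner's ability system tags determine the current facing.
; Facing lookups prefer the data asset, keyed by pawn and CID, over a defaults function stubbed here alongside GetCID and GetActionMontage.
; RefreshEmotes removes both category sets from the controller's external emote component, then adds the current facing's set.
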(event EventBeginPlay)

(event EventTick
  (params (DeltaSeconds :type Float%20%28single-precision%29)))

(fn ClearEmotes
  (Utilities|FlowControl|Sequence
    (:then_0
      (for _array_element (Variables|Default|GetSelfieEmotes)
        (Utilities|Array|Add (LocalVariables|GetCategoryNames) (Utilities|Struct|BreakExternalEmoteCategory _array_element))))
    (:then_1
      (for _array_element_2 (Variables|Default|GetBackEmotes)
        (Utilities|Array|Add (LocalVariables|GetCategoryNames) (Utilities|Struct|BreakExternalEmoteCategory _array_element_2))))
    (:then_2
      (Class|FortControllerComponentExternalEmote|RemoveExternalEmoteCategories (Actor|GetComponentbyClass (Utility|GetOwningController) "/Script/FortniteGame.FortControllerComponent_ExternalEmote") (LocalVariables|GetCategoryNames)))))

(fn GetActionMontage
  (params (ActionTag :type Gameplay%20Tag%20Structure))
  (return 0 "None"))

(fn GetCID
  (return 0))

(fn GetCurrentFacingData
  (bind (_facingdata _isvalid) (FacingData|GetFacingData (Utility|GetCurrentFacingTag)))
  (return _facingdata _isvalid))

(fn GetCurrentFacingTag
  (if (GameplayTags|HasMatchingGameplayTag (Ability|GetAbilitySystemComponent (Components|GetOwner)) "(TagName=\"Ability.PhotoMode.Camera.Facing.Selfie\")")
    (return "(TagName=\"Ability.PhotoMode.Camera.Facing.Selfie\")")
    (else
      (return "(TagName=\"Ability.PhotoMode.Camera.Facing.Back\")"))))

(fn GetFacingData
  (params (FacingTag :type Gameplay%20Tag%20Structure))
  (bind (_facingdata _isvalid) (FacingData|GetFacingDatafromAsset FacingTag))
  (bind (_facingdata_2 _isvalid_2) (FacingData|GetFacingDatafromDefaults FacingTag))
  (if _isvalid
    (return _facingdata true)
    (else
      (return _facingdata_2 _isvalid_2))))

(fn GetFacingDataFromAsset
  (params (FacingTag :type Gameplay%20Tag%20Structure))
  (bind (_data _returnvalue) (PhotoMode|GetSelfieFacingData (Variables|FacingData|GetFacingDataAsset) (Utility|GetFortPlayerAthenaPawn) (Utility|GetCID)))
  (bind (_data_2 _returnvalue_2) (PhotoMode|GetBackFacingData (Variables|FacingData|GetFacingDataAsset) (Utility|GetFortPlayerAthenaPawn) (Utility|GetCID)))
  (Utilities|IsValid (Variables|FacingData|GetFacingDataAsset)
    (:Is%20Valid
      (switch Utilities|FlowControl|Switch|SwitchonGameplayTag (Variables|GetFacingTag)
        (:Ability.PhotoMode.Camera.Facing.Back
          (return _data_2 _returnvalue_2))
        (:Ability.PhotoMode.Camera.Facing.Selfie
          (return _data _returnvalue))))))

(fn GetFacingDataFromDefaults
  (params (FacingTag :type Gameplay%20Tag%20Structure))
  (return 0 false))

(fn GetFortPlayerAthenaPawn
  (return (Utilities|Casting|CastToFortPlayerPawnAthena (Components|GetOwner))))

(fn GetOwningController
  (return (Fortnite|GetFortPlayerControllerfromActor (Components|GetOwner))))

(fn GetSelfieStickClass
  (return (Variables|Default|GetSelfieStickClass)))

(fn RefreshEmotes
  (Emotes|ClearEmotes)
  (bind _aspawn (Utilities|Casting|CastToPawn (Components|GetOwner)))
  (Class|FortControllerComponentExternalEmote|AddExternalEmoteCategories (Actor|GetComponentbyClass (Pawn|GetController _aspawn) "/Script/FortniteGame.FortControllerComponent_ExternalEmote") (select (== (Utility|GetCurrentFacingTag) "(TagName=\"Ability.PhotoMode.Camera.Facing.Selfie\")") (Variables|Default|GetSelfieEmotes) (Variables|Default|GetBackEmotes)) false true))
