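; B_PhotoMode_BPFL is a Blueprint function library of shared helpers for the game's photo mode.
; It reads the photo config from the controller's photo camera component and builds new photo metadata with a GUID, orientation and merged gameplay tags.
; Show and hide functions broadcast a stateful UI message through the global GameplayEventRouter.
; Other helpers fetch the avatar's photo mode animation component and classify dimensions as 16:9, square or 9:16.
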(fn CreateNewPhotoMetadata
  (params (PlayerController :type Player%20Controller%20Object%20Reference) (AdditionalTags :type Gameplay%20Tag%20Container%20Structure%20%28by%20ref%29))
  (bind _config (CallFunction|GetPhotoConfig PlayerController PlayerController))
  (bind _returnvalue (GameplayTags|MakeLiteralGameplayTagContainer AdditionalTags))
  (GameplayTags|AppendGameplayTagContainers _returnvalue (Variables|/Script/PhotoModeRuntime.PhotoConfig|GetTags _config))
  (return (Utilities|Struct|MakePhotoMetadata (Guid|NewGuid) (Variables|/Script/PhotoModeRuntime.PhotoConfig|GetOrientation _config) _returnvalue)))

(fn GetPhotoConfig
  (params (PlayerController :type Player%20Controller%20Object%20Reference))
  (return (Class|ControllerCompPhotoCamera|GetPhotoConfig (Actor|GetComponentbyClass PlayerController "/Script/PhotoModeRuntime.ControllerComp_PhotoCamera"))))

(fn GetPhotoModeAnimComponentFromActorInfo
  (params (Target :type Gameplay%20Ability%20Object%20Reference))
  (bind _returnvalue (Ability|GetAvatarActorfromActorInfo Target))
  (Utilities|IsValid Target
    (:Is%20Valid
      (Utilities|IsValid _returnvalue
        (:Is%20Valid
          (return (Actor|GetComponentbyClass _returnvalue "/PhotoMode/Blueprints/B_PM_PlayerPawnAnim_Component.B_PM_PlayerPawnAnim_Component_C")))))))

(fn HidePhotoCameraHUD
  (params (PlayerController :type Player%20Controller%20Object%20Reference))
  (GameplayEventRouter|BroadcastStatefulEvent (GameplayEventRouter|GetGlobalEventRouter PlayerController) (Utilities|Struct|MakeSPMPlayerPhotoModeEnteredUIMessage false) PlayerController)
  (return))

(fn Photo_CheckAspectRatio
  (params (X :type Integer) (Y :type Integer))
  (LocalVariables|SetLocal_X (Math|Conversions|ToFloat(Integer) X))
  (LocalVariables|SetLocal_Y (Math|Conversions|ToFloat(Integer) Y))
  (if (Math|Float|NearlyEqual(Float) (LocalVariables|GetLocal_X) (LocalVariables|GetLocal_Y) 1e+01)
    (return 1)
    (elif (Math|Float|NearlyEqual(Float) (Math|Float|SafeDivide (LocalVariables|GetLocal_X) 16.0) (Math|Float|SafeDivide (LocalVariables|GetLocal_Y) 9.0) 0.01)
      ; 0 = 16:9
      ; 1 = Square
      ; 2 = 9:16
      (return 0)
      (elif (> (Math|Float|SafeDivide (LocalVariables|GetLocal_Y) 9.0) (Math|Float|SafeDivide (LocalVariables|GetLocal_X) 16.0))
        (return 2)
        (else
          (return 0))))))

(fn ShowPhotoCameraHUD
  (params (PlayerController :type Player%20Controller%20Object%20Reference))
  (GameplayEventRouter|BroadcastStatefulEvent (GameplayEventRouter|GetGlobalEventRouter PlayerController) (Utilities|Struct|MakeSPMPlayerPhotoModeEnteredUIMessage true) PlayerController)
  (return))
